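# Niagara system for the Hera Desert environment FX, named for a paint flow effect but containing no visible effect.
# It holds a single emitter named Empty with no renderers, and System State is the only module listed, so it draws nothing.
# It exposes user parameters for front-end and front-end preview flags and a zero-default position offset, with no visible logic consuming them.
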NiagaraSystem /Game/Environments/Hera/FX/Desert/NSLW_Desert_Paint_Flow
dependencies {usedDataInterfaces [], usedDynamicInputs [], usedModules [&SystemState]}
emitters {Empty {rendererClasses [], renderers []}}
summary {
 systemName=NSLW_Desert_Paint_Flow
 emitters [{emitterName=Empty, rendererClasses []}]
 userVariables [
  User.bIsFrontEnd {type=NiagaraBool}
  User.bIsFrontEndPreview {type=NiagaraBool}
  User.PositionOffset {type=Vector3f, description="", defaultValue {struct=Vector3f, value=<0 0 0>}}
 ]
}
