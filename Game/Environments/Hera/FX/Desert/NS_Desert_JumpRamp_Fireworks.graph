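# Niagara system for desert jump-ramp fireworks, firing bursts of flame and sparks from two lines 500 units either side of center.
# All three emitters pick an X side and index a Y-offset array by normalized loop age, stepping spawn points along the ramp.
# Mesh flames and GPU SubUV fire-lick sprites form the fire.
# Dense GPU sparks add gravity, drag, curl noise, collision, waveform flicker and distance-based spawn scaling.
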NiagaraSystem /Game/Environments/Hera/FX/Desert/NS_Desert_JumpRamp_Fireworks
dependencies {
 usedDataInterfaces [
  &NiagaraDataInterfaceCurve
  &NiagaraDataInterfaceColorCurve
  &NiagaraDataInterfaceArrayFloat3
  &NiagaraDataInterfaceArrayFloat
 ]
 usedDynamicInputs [&SimulationPosition, &Vector2DFromFloat]
 usedModules [&SystemState]
}
emitters {
 FireLicks_Sprites {
  simTarget=GPUComputeSim
  emitterUpdateScript [
   &EmitterState()
   &SpawnRate(SpawnRate=100, Use Spawn Probability=false) {module=SpawnRate001}
   &SpawnRate(
    SpawnRate=/Niagara/DynamicInputs/Math/LerpMultipleFloats(
     Float Count:ENiagara_MultipleLerpCount=::NewEnumerator0
     1:NiagaraFloat=&UniformRangedFloat(Minimum=15, Maximum=20)
     2:NiagaraFloat=6
     3:NiagaraFloat=0
     4:NiagaraFloat=?
     5:NiagaraFloat=?
     Alpha:NiagaraFloat=&ClampFloat(Float=&FN_LOD_Distance_Remap(Start Distance=600, End Distance=2400, Minimum Output=0, Maximum Output=1))
    )
    Use Spawn Probability=false
   ) {enabled=false}
  ]
  particleSpawnScript [
   &InitializeParticle(
    Color=<rgba 4 4 4 1>
    Lifetime=&RandomRangeFloat(Minimum=1, Maximum=1.5)
    Mass Mode=::NewEnumerator1
    Sprite Rotation Angle=&UniformRangedFloat(Minimum=0, Maximum=360)
    Sprite Rotation Mode=::NewEnumerator0
    Sprite Size Mode=::NewEnumerator1
    Uniform Sprite Size Max=600
    Uniform Sprite Size Min=250
    UsePositionOffset=false
    Write Material Random=true
   ) {module=InitializeParticle001}
   &Color(Color=<rgba 10 0.6543 0 0.55>)
   &DynamicMaterialParameters(
    Index 0 Param 2=&RandomRangeFloat(Minimum=0, Maximum=1)
    Index 0 Param 4=2.5
    Module.Param0WriteEnabled=false
    Module.Param1WriteEnabled=true
    Module.Param2WriteEnabled=false
    Module.Param3WriteEnabled=true
   ) {module=DynamicMaterialParameters001}
   &ShapeLocation(
    MaskSpawnGroup=false
    Offset=&MakeVector(
     X=&SelectFloatFromArray(
      Array Sampling Mode=::NewEnumerator0
      Direct Array Index=?
      Float Selection Array=DataInterface() {propertyValues {FloatData [-500,500]}}
      Random Seed=0
     )
     Y=&SelectFloatFromArray(
      Direct Array Index=&MakeIntFromFloat(
       Float=&FloatFromCurve(FloatCurve=DataInterface() {propertyValues {Curve {keys [<at 0 0>,<at 1 10>]}}}, CurveIndex=$System.NormalizedLoopAge)
       Conversion Mode=::NewEnumerator3
      )
      Float Selection Array=DataInterface() {propertyValues {FloatData [200,-300,-800,-1200,-1600]}}
      Random Seed=0
     )
     Z=-100
    )
    Offset Mode=::NewEnumerator0
    Override Seed=false
    Sphere Radius=25
    UseOwnerScale=true
   )
   &AddVelocity(
    Module.Use Falloff Distance=false
    Origin Offset=&MakeVector(
     X=0
     Y=&FloatFromCurve(FloatCurve=DataInterface() {propertyValues {Curve {keys [<at 0 0>,<at 1 -1>]}}}, CurveIndex=$System.NormalizedLoopAge, Scale Curve=-1500)
     Z=-2500
    )
    Override Seed=false
    Velocity Mode=::NewEnumerator1
    Velocity Speed=&RandomRangeFloat(Minimum=100, Maximum=1000)
   ) {module=AddVelocity001}
  ]
  particleUpdateScript [
   /Niagara/Modules/Update/SubUV/SubUVAnimation(Number Of Frames=&FloatFromCurve(FloatCurve=DataInterface() {propertyValues {Curve {keys [{arriveTangent=5.9363, interpMode=RCIM_Cubic, leaveTangent=5.9361, tangentMode=RCTM_User, time=0, value=56}, <at 0.9 60>]}}}))
   &ParticleState()
   &ScaleColor(
    Scale Alpha=&Multiply_Float(
     A=0.8
     B=&FloatFromCurve(
      FloatCurve=DataInterface() {propertyValues {Curve {keys [
          {arriveTangent=10.7702, interpMode=RCIM_Cubic, leaveTangent=10.7702, tangentMode=RCTM_User, time=0, value=0.1}
          {arriveTangent=2.2098, interpMode=RCIM_Cubic, leaveTangent=2.2098, time=0.1, value=0.9}
          {arriveTangent=-0.0116, interpMode=RCIM_Cubic, leaveTangent=-0.0116, tangentMode=RCTM_User, time=0.4073, value=1}
          {arriveTangent=-2.0743, interpMode=RCIM_Cubic, leaveTangent=-2.0743, tangentMode=RCTM_User, time=0.6746, value=0.6742}
          {interpMode=RCIM_Cubic, leaveTangent=-2.371, time=1, value=0}
         ]}}}
      Scale Curve=0.8
     )
    )
    Scale RGB=&VectorFromFloat(Value=&FloatFromCurve(FloatCurve=DataInterface() {propertyValues {Curve {keys [<at 0 1>,<at 1 0>]}}}, Scale Curve=25))
    ScaleA=true
    ScaleRGB=true
   ) {module=ScaleColor001}
   &DynamicMaterialParameters(
    Index 0 Param 1=&FloatFromCurve(FloatCurve=DataInterface() {propertyValues {Curve {keys [<at 0 1>]}}})
    Index 0 Param 2=0.5
    Index 0 Param 3=&FloatFromCurve(FloatCurve=DataInterface() {propertyValues {Curve {keys [<at 0 0.5 RCIM_Cubic>, {arriveTangent=1.5, interpMode=RCIM_Cubic, leaveTangent=1.5, time=0.5235, value=0.5}, <at 1 2>]}}})
    Module.Param0WriteEnabled=true
    Module.Param1WriteEnabled=false
    Module.Param2WriteEnabled=true
    Module.Param3WriteEnabled=false
   )
   &SpriteRotationRate(Rotation Rate=&Lerp_Float(A=-10, B=10, Alpha=$Particles.Initial.MaterialRandom)) {module=SpriteRotationRate001}
   &CameraOffset(Camera Offset Amount=-2)
   &SolveForcesAndVelocity(Clamp Velocity=false, Limit Acceleration=false)
  ]
  rendererClasses [&NiagaraSpriteRendererProperties]
  renderers [{rendererClass=&NiagaraSpriteRendererProperties}]
 }
 Flames {
  emitterUpdateScript [
   &EmitterState()
   &SpawnRate(SpawnRate=50, Use Spawn Probability=false)
  ]
  particleSpawnScript [
   &InitializeParticle(
    Color=&ColorFromCurve(
     ColorCurve=DataInterface() {propertyValues {
       AlphaCurve {keys [<at 0 1 RCIM_Cubic>,<at 0.9977 1 RCIM_Cubic>]}
       BlueCurve {keys [<at 0.0119 0 RCIM_Cubic>,<at 0.2536 0>]}
       GreenCurve {keys [<at 0.0119 0.1859 RCIM_Cubic>,<at 0.2536 0.2325>]}
       RedCurve {keys [<at 0.0119 1 RCIM_Cubic>,<at 0.2536 1>]}
      }}
     CurveIndex=$System.NormalizedLoopAge
    )
    Lifetime Max=1
    Lifetime Min=0.8
    Lifetime Mode=::NewEnumerator1
    Mesh Scale Mode=::NewEnumerator1
    Mesh Uniform Scale Max=12
    Mesh Uniform Scale Min=8
    UsePositionOffset=false
    Write Material Random=true
   )
   &InitialMeshOrientation(Mesh Orientation Mode=::NewEnumerator2, Use Rotation Vector=false)
   &AddVelocity(
    Velocity=&Multiply_Vector(
     A=&SelectVectorFromArray(
      Direct Array Index=$Particles.PartitionParticles.Partition
      Vector Selection Array=DataInterface() {propertyValues {FloatData [<1500 250 0>, <1500 250 0>, <-1500 250 0>, <-1500 250 0>]}}
     )
     B=&RandomRangeVector(Minimum=<0 0 1>, Maximum=<0 0 3>)
    )
    Velocity Speed Scale=0.2
   ) {enabled=false}
   &ShapeLocation(
    MaskSpawnGroup=false
    Offset=&MakeVector(
     X=&SelectFloatFromArray(
      Array Sampling Mode=::NewEnumerator0
      Direct Array Index=?
      Float Selection Array=DataInterface() {propertyValues {FloatData [-500,500]}}
      Random Seed=0
     )
     Y=&SelectFloatFromArray(
      Direct Array Index=&MakeIntFromFloat(
       Float=&FloatFromCurve(FloatCurve=DataInterface() {propertyValues {Curve {keys [<at 0 0>,<at 1 10>]}}}, CurveIndex=$System.NormalizedLoopAge)
       Conversion Mode=::NewEnumerator3
      )
      Float Selection Array=DataInterface() {propertyValues {FloatData [200,-250,-700,-1100,-1400]}}
      Random Seed=0
     )
     Z=-400
    )
    Offset Mode=::NewEnumerator0
    Override Seed=false
    Sphere Radius=25
    UseOwnerScale=true
   )
   &AddVelocity(
    Module.Use Falloff Distance=false
    Origin Offset=&MakeVector(
     X=0
     Y=&FloatFromCurve(FloatCurve=DataInterface() {propertyValues {Curve {keys [<at 0 0>,<at 1 -1>]}}}, CurveIndex=$System.NormalizedLoopAge, Scale Curve=0)
     Z=-3000
    )
    Override Seed=false
    Velocity Mode=::NewEnumerator1
    Velocity Speed=&RandomRangeFloat(Minimum=200, Maximum=250)
   ) {module=AddVelocity001}
  ]
  particleUpdateScript [
   &ParticleState()
   &ScaleMeshSize(Scale Factor=&Multiply_VectorByFloat(
     Vector=&Multiply_VectorByFloat(
      Vector=<1.25 0.25 0.25>
      Float=&FloatFromCurve(FloatCurve=DataInterface() {propertyValues {Curve {keys [
           {arriveTangent=4.1557, interpMode=RCIM_Cubic, leaveTangent=4.1557, tangentMode=RCTM_User, time=0, value=0}
           {arriveTangent=1.2805, interpMode=RCIM_Cubic, leaveTangent=1.2805, tangentMode=RCTM_User, time=0.3, value=0.75}
           {arriveTangent=-0.3349, interpMode=RCIM_Cubic, leaveTangent=-0.3349, tangentMode=RCTM_User, time=1, value=1.175}
          ]}}})
     )
     Float=1
    ))
   &DynamicMaterialParameters(
    Index 0 Param 1=&FloatFromCurve(FloatCurve=DataInterface() {propertyValues {Curve {keys [
         <at 0 0>
         {interpMode=RCIM_Cubic, tangentMode=RCTM_User, time=0.7, value=0}
         {arriveTangent=1.7755, interpMode=RCIM_Cubic, leaveTangent=1.7755, tangentMode=RCTM_User, time=1, value=0.625}
        ]}}})
    Index 0 Param 2=&FloatFromCurve(
     FloatCurve=DataInterface() {propertyValues {Curve {keys [
         {interpMode=RCIM_Cubic, tangentMode=RCTM_User, time=0, value=0.1}
         {arriveTangent=2.5262, interpMode=RCIM_Cubic, leaveTangent=2.5262, tangentMode=RCTM_User, time=0.4316, value=0.9567}
         {interpMode=RCIM_Cubic, tangentMode=RCTM_User, time=0.6, value=1.25}
         {interpMode=RCIM_Cubic, tangentMode=RCTM_User, time=0.9, value=1.225}
         {interpMode=RCIM_Cubic, tangentMode=RCTM_User, time=1, value=1}
        ]}}}
     Scale Curve=500
    )
    Index 0 Param 3=1
    Index 0 Param 4=0.25
    Index 1 Param 2=0.5
    Index 1 Param 4=0.5
    Index 2 Param 1=1
    Index 2 Param 2=&FloatFromCurve(
     FloatCurve=DataInterface() {propertyValues {Curve {keys [<at 0 0.05>, {interpMode=RCIM_Cubic, tangentMode=RCTM_User, time=0.5, value=0.95}, <at 1 0.7>]}}}
     Scale Curve=4
    )
    Index 2 Param 3=1
    Index 2 Param 4=1
    Index 3 Param 1=&FloatFromCurve(
     FloatCurve=DataInterface() {propertyValues {Curve {keys [
         <at 0 0.25>
         {arriveTangent=1.1168, interpMode=RCIM_Cubic, leaveTangent=1.1168, tangentMode=RCTM_User, time=0.4875, value=0.7}
         {arriveTangent=1.5356, interpMode=RCIM_Cubic, leaveTangent=1.5356, tangentMode=RCTM_User, time=0.7137, value=0.9086}
         <at 1 1>
        ]}}}
     Scale Curve=0.9
    )
    Index 3 Param 2=&RandomRangeFloat(Minimum=0, Maximum=1)
    Module.Param0WriteEnabled=true
    Module.Param0WriteEnabled1=true
    Module.Param0WriteEnabled2=true
    Module.Param0WriteEnabled3=true
    Module.Param1WriteEnabled=true
    Module.Param1WriteEnabled1=true
    Module.Param1WriteEnabled2=true
    Module.Param1WriteEnabled3=true
    Module.Param2WriteEnabled=true
    Module.Param2WriteEnabled1=true
    Module.Param2WriteEnabled2=true
    Module.Param2WriteEnabled3=true
    Module.Param3WriteEnabled=true
    Module.Param3WriteEnabled1=true
    Module.Param3WriteEnabled2=true
    Module.Param3WriteEnabled3=true
    Write Parameter Index 1=true
    Write Parameter Index 2=true
    Write Parameter Index 3=true
   ) {module=DynamicMaterialParameters001}
   &ScaleColor(
    Scale Alpha=&FloatFromCurve(
     FloatCurve=DataInterface() {propertyValues {Curve {keys [
         {interpMode=RCIM_Cubic, tangentMode=RCTM_User, time=0, value=0}
         {interpMode=RCIM_Cubic, tangentMode=RCTM_User, time=0.5, value=1}
         {tangentMode=RCTM_User, time=1, value=0}
        ]}}}
     Scale Curve=0.5
    )
    Scale RGB=&Multiply_VectorByFloat(
     Vector=<50 50 50>
     Float=&FloatFromCurve(
      FloatCurve=DataInterface() {propertyValues {Curve {keys [
          {interpMode=RCIM_Cubic, tangentMode=RCTM_User, time=0, value=0.2}
          {interpMode=RCIM_Cubic, tangentMode=RCTM_User, time=0.5, value=2}
          <at 1 0.4>
         ]}}}
      Scale Curve=50
     )
    )
    ScaleA=true
    ScaleRGB=true
   )
   &SolveForcesAndVelocity(Clamp Velocity=false, Limit Acceleration=false)
   /SetVariables(
    Particles.CamOff2=5
    Particles.Linear Color=&MultiplyLinearColor(Linear Color=$Particles.Color, Scale Factor=<0.2 0.2 0.2 50>)
    Particles.MaskedScale:Vector3f=&Multiply_Vector(A=$Particles.Scale, B=<0.8 0.8 0.8>)
   ) {isSetParameters=true}
  ]
  rendererClasses [&NiagaraMeshRendererProperties]
  renderers [
   {rendererClass=&NiagaraMeshRendererProperties, rendererIndex=0}
   {rendererClass=&NiagaraMeshRendererProperties, rendererIndex=1}
  ]
 }
 Sparks_EPIC {
  simTarget=GPUComputeSim
  emitterUpdateScript [
   &EmitterState(
    Enable Distance Culling=true
    MaxDistance=25000
    Module.bApplySpawnCountScaleByDistance=true
    Module.bUseMaxDistance=true
    Module.bUseMinDistance=false
    Scalability Mode=::NewEnumerator1
    Scale Spawn Count=true
    Spawn Count Scale By Distance=DataInterface() {propertyValues {Curve {keys [<at 0 1>,<at 0.5 1>,<at 1 0>]}}}
   )
   &SpawnRate(SpawnRate=25000, Use Spawn Probability=false)
  ]
  particleSpawnScript [
   &InitializeParticle(
    Color=&MultiplyLinearColor(
     Linear Color=&ColorFromCurve(
      ColorCurve=DataInterface() {propertyValues {
        AlphaCurve {keys [<at 0 1 RCIM_Cubic>,<at 0.9977 1 RCIM_Cubic>]}
        BlueCurve {keys [<at 0.0119 0 RCIM_Cubic>,<at 0.2536 0>,<at 0.3089 0>]}
        GreenCurve {keys [<at 0.0119 0.1768 RCIM_Cubic>,<at 0.2536 0.2281>,<at 0.3089 0.2828>]}
        RedCurve {keys [<at 0.0119 1 RCIM_Cubic>,<at 0.2536 1>,<at 0.3089 1>]}
       }}
      CurveIndex=$System.NormalizedLoopAge
     )
     Scale Factor=&MakeVector4_FromFloat(Float=&RandomRangeFloat(Minimum=15, Maximum=100))
    )
    Lifetime Mode=::NewEnumerator1
    Sprite Rotation Mode=::NewEnumerator2
    Sprite Size Max=<250 250>
    Sprite Size Min=<50 250>
    Sprite Size Mode=::NewEnumerator2
    UsePositionOffset=false
    Write Material Random=true
   )
   &AddVelocity(Velocity=&Multiply_Vector(
     A=&SelectVectorFromArray(
      Direct Array Index=$Particles.PartitionParticles.Partition
      Vector Selection Array=DataInterface() {propertyValues {FloatData [<1500 250 800>, <1500 250 250>, <-1500 250 800>, <-1500 250 250>]}}
     )
     B=&RandomRangeVector(Minimum=<-0.25 -0.25 0.2>, Maximum=<0.25 1 1>)
    ))
   &ShapeLocation(
    MaskSpawnGroup=false
    Offset=&MakeVector(
     X=&SelectFloatFromArray(
      Array Sampling Mode=::NewEnumerator0
      Direct Array Index=?
      Float Selection Array=DataInterface() {propertyValues {FloatData [-500,500]}}
      Random Seed=0
     )
     Y=&SelectFloatFromArray(
      Direct Array Index=&MakeIntFromFloat(
       Float=&FloatFromCurve(FloatCurve=DataInterface() {propertyValues {Curve {keys [<at 0 0>,<at 1 10>]}}}, CurveIndex=$System.NormalizedLoopAge)
       Conversion Mode=::NewEnumerator3
      )
      Float Selection Array=DataInterface() {propertyValues {FloatData [200,-300,-800,-1200,-1700]}}
      Random Seed=0
     )
     Z=-400
    )
    Offset Mode=::NewEnumerator0
    Override Seed=false
    Sphere Radius=25
    UseOwnerScale=true
   )
   &AddVelocity(
    Module.Use Falloff Distance=false
    Origin Offset=<0 0 -1500>
    Override Seed=false
    Velocity Mode=::NewEnumerator1
    Velocity Speed=&RandomRangeFloat(Minimum=250, Maximum=2500)
   ) {module=AddVelocity001}
  ]
  particleUpdateScript [
   &Drag(
    Drag=&FloatFromCurve(FloatCurve=DataInterface() {propertyValues {Curve {keys [{tangentMode=RCTM_User, time=0, value=0}, {tangentMode=RCTM_User, time=0.5, value=1}]}}})
    Use Linear Drag=true
    Use Rotational Drag=false
   )
   &GravityForce(Gravity=<0 0 -1600>)
   &CurlNoiseForce(Noise Frequency=5, Noise Strength=500, Pan Noise=false)
   &Collision(
    Age Colliding Particles=false
    Collision Enabled=&SetBoolByFloatComparison(
     A=&FloatFromCurve(FloatCurve=DataInterface() {propertyValues {Curve {keys [<at 0.1 0>,<at 0.2 2>]}}})
     B=0.01
    )
    Friction=0.2
    GPU Collision Type=::NewEnumerator2
    KillOnCollision=false
    Module.Control Rest State Via Surface Normal=false
    Module.EnableMaxCollisionCount=false
    Module.Randomize Collision Normal Vector=false
    Restitution=&RandomRangeFloat(Minimum=0.25, Maximum=0.25)
   )
   &SolveForcesAndVelocity(Clamp Velocity=false, Limit Acceleration=false)
   &ParticleState()
   &DynamicMaterialParameters(
    Index 0 Param 1=0.8
    Index 0 Param 2=1
    Index 0 Param 3=1
    Index 1 Param 1=150
    Index 1 Param 4=1
    Module.Param0WriteEnabled=true
    Module.Param0WriteEnabled1=true
    Module.Param1WriteEnabled=true
    Module.Param1WriteEnabled1=true
    Module.Param2WriteEnabled=true
    Module.Param2WriteEnabled1=true
    Module.Param3WriteEnabled=false
    Module.Param3WriteEnabled1=true
    Write Parameter Index 1=true
   )
   &ScaleColor(
    Scale Alpha=0.01
    Scale RGB=&Multiply_VectorByFloat(
     Vector=<5 5 5>
     Float=&FloatFromCurve(
      FloatCurve=DataInterface() {propertyValues {Curve {keys [{interpMode=RCIM_Cubic, tangentMode=RCTM_User, time=0, value=1}, {interpMode=RCIM_Cubic, tangentMode=RCTM_User, time=1, value=0.5}]}}}
      Scale Curve=&Waveform(
       "Waveform [1]"=::NewEnumerator0
       "[1] Phase Scale"=&RandomRangeFloat(Minimum=2, Maximum=4)
       "[1] Frequency"=&RandomRangeFloat(Minimum=4, Maximum=5)
       "[1] Amplitude Min/Max"=<1 4>
      )
     )
    )
    ScaleA=true
    ScaleRGB=true
   ) {module=ScaleColor001}
   /Niagara/Modules/Update/Camera/MaintainInCameraParticleScale(Half FOV=1)
   &ScaleSpriteSizeBySpeed(
    Max Scale Factor=<0.05 0.2>
    Min Scale Factor=<0.05 0.1>
    Sample Scale Factor By Curve=true
    Scale Factor Curve=DataInterface() {propertyValues {Curve {keys [
        {interpMode=RCIM_Cubic, tangentMode=RCTM_User, time=0, value=0}
        {arriveTangent=2.9664, interpMode=RCIM_Cubic, leaveTangent=2.9664, tangentMode=RCTM_User, time=1, value=1}
       ]}}}
    Velocity Threshold=250
   )
   &ScaleSpriteSize(
    Uniform Curve Scale=&Multiply_Float(
     A=&Multiply_Float(
      A=0.2
      B=&Waveform(
       "Waveform [1]"=::NewEnumerator0
       "[1] Phase Scale"=&RandomRangeFloat(Minimum=3, Maximum=10)
       "[1] Frequency"=&RandomRangeFloat(Minimum=0.2, Maximum=0.5)
       "[1] Amplitude Min/Max"=<0 1>
      )
     )
     B=&FN_LOD_Distance_Remap(Start Distance=2500, End Distance=10000, Minimum Output=1, Maximum Output=2, Clamp Output=true)
    )
    Uniform Curve Sprite Scale=DataInterface() {propertyValues {Curve {keys [
        {arriveTangent=2.9868, interpMode=RCIM_Cubic, leaveTangent=2.9868, tangentMode=RCTM_User, time=0, value=1.75}
        <at 0.4 2.3>
        {arriveTangent=-0.0405, interpMode=RCIM_Cubic, leaveTangent=-0.0405, tangentMode=RCTM_User, time=1, value=1.4}
       ]}}}
   )
   /SetVariables(
    Particles.Linear Color=&MultiplyLinearColor(Linear Color=$Particles.Color, Scale Factor=<0.02 0.02 0.02 0.01>)
    Particles.Vector:Vector3f=&Multiply_Vector(A=$Particles.Velocity, B=<1 1 1>)
    Particles.Vector 2D:Vector2f=&Multiply_Vector2D(A=<1 1.5>, B=$Particles.SpriteSize)
   ) {isSetParameters=true}
  ]
  rendererClasses [&NiagaraSpriteRendererProperties]
  renderers [
   {rendererClass=&NiagaraSpriteRendererProperties}
   {rendererClass=&NiagaraSpriteRendererProperties, rendererIndex=1}
  ]
 }
}
summary {
 systemName=NS_Desert_JumpRamp_Fireworks
 emitters [
  {emitterName=Sparks_EPIC, simTarget=GPUComputeSim, rendererClasses [&NiagaraSpriteRendererProperties]}
  {emitterName=Flames, rendererClasses [&NiagaraMeshRendererProperties]}
  {emitterName=FireLicks_Sprites, simTarget=GPUComputeSim, rendererClasses [&NiagaraSpriteRendererProperties]}
 ]
 userVariables []
}
